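; Anim notify for BlondeJaw emote effects that line-traces for obstructions and feeds the results to its spawned particle system.
; It rotates the effect to the mesh plus an offset, then traces along each mesh-rotated velocity direction, mapping hit time to a delay.
; One direction sets HitGround and ExplosionDelay and moves the effect to the impact or partway along the trace.
; Multiple directions pack delays into CustomLifetime's X/Y/Z by index.
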(fn Received_Notify
  (params (MeshComp :type Skeletal%20Mesh%20Component%20Object%20Reference) (Animation :type Anim%20Sequence%20Base%20Object%20Reference) (EventReference :type Anim%20Notify%20Event%20Reference%20Structure%20%28by%20ref%29))
  (bind _returnvalue (Transformation|GetWorldRotation MeshComp))
  (bind _velocitydirections (Variables|AnimNotify|GetVelocityDirections))
  (LocalVariables|SetMultipleVectors (> (Utilities|Array|Length (Variables|AnimNotify|GetVelocityDirections)) 1))
  (bind _returnvalue_2 (AnimNotify|GetSpawnedEffect))
  (Transformation|SetWorldRotation _returnvalue_2 _returnvalue false false)
  (Transformation|AddRelativeRotation _returnvalue_2 (Variables|AnimNotify|GetRotationOffset) false false)
  (for _index (range (Utilities|Array|Length (Variables|AnimNotify|GetVelocityDirections)))
    (bind (_outhit _returnvalue_3) (Collision|LineTraceByChannel (+ (Transformation|GetWorldLocation _returnvalue_2) (* (Math|Vector|RotateVector (Utilities|Array|Get(acopy) _velocitydirections _index) _returnvalue) 0.5)) (+ (+ (Transformation|GetWorldLocation _returnvalue_2) (* (Math|Vector|RotateVector (Utilities|Array|Get(acopy) _velocitydirections _index) _returnvalue) 0.5)) (* (Math|Vector|RotateVector (Utilities|Array|Get(acopy) _velocitydirections _index) _returnvalue) 3.5))))
    (if (LocalVariables|GetMultipleVectors)
      (LocalVariables|SetTempVector (+ (LocalVariables|GetTempVector) (Utilities|Select (Math|Vector|MakeVector (Math|Float|Lerp 0.5 4.0 (.Time (Collision|BreakHitResult _outhit)))) (Math|Vector|MakeVector 0.0 (Math|Float|Lerp 0.5 4.0 (.Time (Collision|BreakHitResult _outhit)))) (Math|Vector|MakeVector 0.0 0.0 (Math|Float|Lerp 0.5 4.0 (.Time (Collision|BreakHitResult _outhit)))) _index)))
      (else
        (Effects|Components|ParticleSystem|SetBoolParameter _returnvalue_2 "HitGround" _returnvalue_3)
        (Effects|Components|ParticleSystem|SetFloatParameter _returnvalue_2 "ExplosionDelay" (Math|Float|Lerp 0.5 4.0 (.Time (Collision|BreakHitResult _outhit))))
        (Transformation|SetWorldLocation _returnvalue_2 (select _returnvalue_3 (.Location (Collision|BreakHitResult _outhit)) (Math|Vector|Lerp(Vector) (+ (Transformation|GetWorldLocation _returnvalue_2) (* (Math|Vector|RotateVector (Utilities|Array|Get(acopy) _velocitydirections _index) _returnvalue) 0.5)) (+ (+ (Transformation|GetWorldLocation _returnvalue_2) (* (Math|Vector|RotateVector (Utilities|Array|Get(acopy) _velocitydirections _index) _returnvalue) 0.5)) (* (Math|Vector|RotateVector (Utilities|Array|Get(acopy) _velocitydirections _index) _returnvalue) 3.5)) 0.2)) false false)
        (return false))))
  (if (LocalVariables|GetMultipleVectors)
    (Effects|Components|ParticleSystem|SetVectorParameter _returnvalue_2 "CustomLifetime" (LocalVariables|GetTempVector))
    (return false)
    (else
      (return false))))
