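# Niagara system for the BlondeJaw emote that surrounds the character with buildings which break apart near its sampled sockets.
# Buildings spawn on a grid in three size classes, follow the ground by traces, scale in with an overshoot curve via region masks, and scroll in the front end.
# Socket proximity drives a death switch with system arrays, and system death position and size trigger size-scaled Smoke and colliding Chunks bursts.
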NiagaraSystem /BRCosmetics/Animation/Game/MainPlayer/Emotes/BlondeJaw/FX/NS_Emote_BlondeJaw_02
dependencies {
 usedDataInterfaces [
  &NiagaraDataInterfaceCurve
  &NiagaraDataInterfaceMeshRendererInfo
  &NiagaraDataInterfaceArrayFloat3
  &NiagaraDataInterfaceCollisionQuery
  &NiagaraDataInterfaceDebugDraw
  &NiagaraDataInterfaceArrayFloat
  &NiagaraDataInterfaceArrayColor
  &NiagaraDataInterfaceArrayInt32
  &NiagaraDataInterfaceArrayPosition
  &NiagaraDataInterfaceSkeletalMesh
  &NiagaraDataInterfaceSocketReader
 ]
 usedDynamicInputs [&MakeIntFromFloat, &FNTimeOfDayColor]
 usedModules [
  &self:SystemSpawnScript.SystemScriptSource.SystemScriptGraph.NiagaraNodeAssignment_3.SetVariables_#1
  &self:SystemSpawnScript.SystemScriptSource.SystemScriptGraph.NiagaraNodeAssignment_1.SetVariables_#2
  &self:SystemSpawnScript.SystemScriptSource.SystemScriptGraph.NiagaraNodeAssignment_5.SetVariables_#3
  &self:SystemSpawnScript.SystemScriptSource.SystemScriptGraph.NiagaraNodeAssignment_0.SetVariables_#4
  &self:SystemSpawnScript.SystemScriptSource.SystemScriptGraph.NiagaraNodeAssignment_6.SetVariables_#5
  &self:SystemSpawnScript.SystemScriptSource.SystemScriptGraph.NiagaraNodeAssignment_2.SetVariables_#6
  &self:SystemSpawnScript.SystemScriptSource.SystemScriptGraph.NiagaraNodeAssignment_7.SetVariables_#7
  &SystemState
  &self:ReadDeath
  &self:SystemSpawnScript.SystemScriptSource.SystemScriptGraph.NiagaraNodeAssignment_4.SetVariables_#8
 ]
}
emitters {
 Buildings {
  emitterUpdateScript [
   &EmitterState(
    Inactive Response=::NewEnumerator1
    Life Cycle Mode=::NewEnumerator1
    Loop Behavior=::NewEnumerator1
    Loop Duration Mode=::NewEnumerator1
    UseLoopDelay=false
   )
   /Niagara/Modules/Spawn/Location/SpawnParticlesInGrid(Z Count=0)
   /SetVariables(Emitter.VolumeRadius=180) {isSetParameters=true}
  ]
  particleSpawnScript [
   &FN_SetSpawnGroup(Number of Spawn Groups=3) {enabled=false}
   &PartitionParticles(Number Of Partitions=3, Partition Mode:ENiagara_PartitionMode=::NewEnumerator5) {module=PartitionParticles001}
   &self:PartitionByBuildingSize(
    Fraction_MediumB=0.35
    Franction_SmallB=0.45
    Franction_TallB=0.2
    LBuilding_IndexMax=3
    MBuilding_IndexMax=10
    MBuilding_IndexMin=4
    SBuilding_IndexMax=18
    SBuilding_IndexMin=11
   ) {module=PartitionByBuildingSize}
   &FN_SetSpawnGroup(Number of Spawn Groups=6) {module=FN_SetSpawnGroup001}
   &InitializeParticle(Mesh Scale Mode=::NewEnumerator3, Position Offset Coordinate Space=::Local, UsePositionOffset=true, Write Material Random=true)
   /SetVariables(Particles.MeshIndex=&MaskIntBySpawnGroup(
     Masked Int=&RandomRangeInt(Minimum=0, Maximum=2)
     Passthrough Int=&MaskIntBySpawnGroup(
      Masked Int=&RandomRangeInt(Minimum=3, Maximum=10)
      Passthrough Int=&RandomRangeInt(Minimum=11, Maximum=18)
      Spawn Group Mask=2
     )
     Spawn Group Mask=1
    )) {enabled=false, isSetParameters=true}
   /Niagara/Modules/Spawn/Location/V2/GridLocation(
    Coordinate Space:ENiagaraCoordinateSpace=::World
    Module.Jitter Inside Grid Cells=false
    Module.Override Random Seed=false
    Normalize Offsets=false
    XYZ Dimensions:Vector3f=&MakeVector(
     X=&Divide_Float(A=$Emitter.VolumeRadius, B=5)
     Y=&Divide_Float(A=$Emitter.VolumeRadius, B=5)
     Z=0
    )
   )
   &InitialMeshOrientation(
    Mesh Orientation Mode=::NewEnumerator2
    Rotation=&SelectVectorFromArray(
     Array Sampling Mode=::NewEnumerator0
     Direct Array Index=?
     Vector Selection Array=DataInterface() {propertyValues {FloatData [<0 0 0>, <0 0 0.25>, <0 0 0.5>]}}
     Random Seed=0
    )
    Use Rotation Vector=true
   )
   /SetVariables(Particles.MeshIndex_Debris:NiagaraInt32=&RandomRangeInt(Minimum=0, Maximum=2)) {isSetParameters=true}
   /SetVariables() {isSetParameters=true}
   /SetVariables() {isSetParameters=true}
   /SetVariables(Particles.IsOutsideMask=true) {isSetParameters=true}
  ]
  particleUpdateScript [
   &ParticleState(Kill Particles When Lifetime Has Elapsed=false, Let Infinitely Lived Particles Die When Emitter Deactivates=true)
   &SkeletalMeshLocation(
    Bone Sampling Mode=::NewEnumerator4
    MaskSpawnGroup=false
    Orientation Sampling=::NewEnumerator1
    Position Sampling=::NewEnumerator2
    Skeletal Mesh=$System.Skeletal Mesh
    Use Sampled Mesh Attribute as Ribbon ID=false
   ) {enabled=false}
   &SocketLocation(
    Orientation Sampling=::NewEnumerator1
    Position Sampling=::NewEnumerator1
    SamplePosOffset=false
    Socket Reader=$System.Socket Reader
    Socket Sampling Mode=::NewEnumerator3
    VelocitySpeedLimit=false
   )
   &SkeletalMeshLocation(
    Bone Sampling Mode=::NewEnumerator4
    MaskSpawnGroup=false
    Orientation Sampling=::NewEnumerator1
    Position Sampling=::NewEnumerator2
    Skeletal Mesh=$System.Skeletal Mesh001
    Use Sampled Mesh Attribute as Ribbon ID=false
   ) {enabled=false, module=SkeletalMeshLocation001}
   &SkeletalMeshLocation(
    Bone / Socket Index=1
    Bone Sampling Mode=::NewEnumerator4
    MaskSpawnGroup=false
    Orientation Sampling=::NewEnumerator1
    Position Sampling=::NewEnumerator2
    Skeletal Mesh=$System.Skeletal Mesh001
    Use Sampled Mesh Attribute as Ribbon ID=false
   ) {enabled=false, module=SkeletalMeshLocation002}
   &SkeletalMeshLocation(
    Bone / Socket Index=2
    Bone Sampling Mode=::NewEnumerator4
    MaskSpawnGroup=false
    Orientation Sampling=::NewEnumerator1
    Position Sampling=::NewEnumerator2
    Skeletal Mesh=$System.Skeletal Mesh001
    Use Sampled Mesh Attribute as Ribbon ID=false
   ) {enabled=false, module=SkeletalMeshLocation003}
   &SkeletalMeshLocation(
    Bone / Socket Index=3
    Bone Sampling Mode=::NewEnumerator4
    MaskSpawnGroup=false
    Orientation Sampling=::NewEnumerator1
    Position Sampling=::NewEnumerator2
    Skeletal Mesh=$System.Skeletal Mesh001
    Use Sampled Mesh Attribute as Ribbon ID=false
   ) {enabled=false, module=SkeletalMeshLocation004}
   &SocketLocation(
    Orientation Sampling=::NewEnumerator1
    Position Sampling=::NewEnumerator1
    SamplePosOffset=false
    Socket Reader=$System.Socket Reader001
    Socket Sampling Mode=::NewEnumerator3
    VelocitySpeedLimit=false
   ) {module=SocketLocation001}
   &SocketLocation(
    Orientation Sampling=::NewEnumerator1
    Position Sampling=::NewEnumerator1
    SamplePosOffset=false
    Socket Index=1
    Socket Reader=$System.Socket Reader001
    Socket Sampling Mode=::NewEnumerator3
    VelocitySpeedLimit=false
   ) {module=SocketLocation002}
   &SocketLocation(
    Orientation Sampling=::NewEnumerator1
    Position Sampling=::NewEnumerator1
    SamplePosOffset=false
    Socket Index=2
    Socket Reader=$System.Socket Reader001
    Socket Sampling Mode=::NewEnumerator3
    VelocitySpeedLimit=false
   ) {module=SocketLocation003}
   &SocketLocation(
    Orientation Sampling=::NewEnumerator1
    Position Sampling=::NewEnumerator1
    SamplePosOffset=false
    Socket Index=3
    Socket Reader=$System.Socket Reader001
    Socket Sampling Mode=::NewEnumerator3
    VelocitySpeedLimit=false
   ) {module=SocketLocation004}
   &SolveForcesAndVelocity(Clamp Velocity=false, Limit Acceleration=false)
   /SetVariables(Particles.PositionCenter:NiagaraPosition=&AddVectorToPosition(Position=$Engine.Owner.Position, Vector=$Emitter.OriginOffset)) {isSetParameters=true}
   &self:ScratchModule(
    Dead:NiagaraBool=$Particles.Dead
    GridSize:NiagaraFloat=&Multiply_Float(A=$Emitter.VolumeRadius, B=2)
    InputPosition:NiagaraPosition=$Particles.PositionCenter
    IsOutsideMask:NiagaraBool=$Particles.IsOutsideMask
    Normal:Vector3f=$Engine.Owner.SystemZAxis
    Orient:Quat4f=$Particles.SkeletalMeshLocation.SampledRotation
   ) {module=ScratchModule}
   /SetVariables(Particles.Position=&MakeVector(
     X=&MakeFloatFromVector(Vector=Linked() {linkedVariable {name=Particles.Position, type=NiagaraPosition}}, Channel=::NewEnumerator0)
     Y=&MakeFloatFromVector(Vector=Linked() {linkedVariable {name=Particles.Position, type=NiagaraPosition}}, Channel=::NewEnumerator1)
     Z=&MakeFloatFromVector(Vector=/Niagara/DynamicInputs/Helpers/SimulationPosition(), Channel=::NewEnumerator2)
    )) {isSetParameters=true}
   /SetVariables(Particles.Position=&AddVectorToPosition(Position=$Particles.Position, Vector=<0 0 100>)) {isSetParameters=true}
   /SetVariables(Particles.Position=&AddVectorToPosition(
     Position=$Particles.Position
     Vector=&FN_FrontEndVector3Modifier(
      Gameplay=<0 0 0>
      Front End=&RotateVectorByQuaternion(
       Vector To Rotate=&Multiply_VectorByFloat(Vector=<0 -40 0>, Float=$Engine.DeltaTime)
       Quaternion=$Output.SocketLocation.SampledRotation
      )
     )
    )) {isSetParameters=true}
   &self:ScratchModule_04(TraceDir:Vector3f=<0 0 -1>, TraceDist=250, bIsFrontEnd:NiagaraBool=$User.bIsFrontEndPreview) {module=ScratchModule_04}
   &RegionMask(
    Invert Region=true
    Offset:Vector3f=$Emitter.OriginOffset
    Region Falloff=1
    Region Radius:NiagaraFloat=$Emitter.VolumeRadius
    UseOffset=true
    UseRotation=false
   )
   &RegionMask(
    Offset:Vector3f=&AddVectorToPosition(
     Position=Linked() {linkedVariable {name=Emitter.OriginOffset, type=Vector3f}}
     Vector=&RotateVectorByQuaternion(Vector To Rotate=<0 -220 0>, Quaternion=$Output.SocketLocation.SampledRotation)
    )
    Region Falloff=1
    Region Radius:NiagaraFloat=$Emitter.VolumeRadius
    UseOffset=true
    UseRotation=false
   ) {module=RegionMask001}
   &Timeline(
    Additional Output:ENiagara_TimelineOutput=::NewEnumerator1
    Additional Output Mode:ENiagara_TimelineOutputMode=::NewEnumerator1
    Float Curve=DataInterface() {propertyValues {Curve {keys [
        {arriveTangent=0.9218, interpMode=RCIM_Cubic, leaveTangent=0.9218, tangentMode=RCTM_User, time=0.35, value=0}
        {arriveTangent=2.875, interpMode=RCIM_Cubic, leaveTangent=2.875, time=0.63, value=0.9}
        {arriveTangent=-0.0156, interpMode=RCIM_Cubic, leaveTangent=-0.0156, tangentMode=RCTM_User, time=0.75, value=1.15}
        {arriveTangent=-0.0297, interpMode=RCIM_Cubic, leaveTangent=-0.0297, tangentMode=RCTM_User, time=0.9, value=0.95}
        {arriveTangent=0.0389, interpMode=RCIM_Cubic, leaveTangent=0.0389, tangentMode=RCTM_User, time=0.96, value=1.025}
        <at 1 1 RCIM_Cubic>
       ]}}}
    Play=&Boolean_NotOperation(Bool=&Boolean_OrOperation(Bool1=$Particles.RegionMask.IsInsideMask, Bool2=$Particles.RegionMask001.IsInsideMask))
    Play Rate=0.75
   )
   /SetVariables(Particles.IsOutsideMask:NiagaraBool=&SetBoolByFloatComparison(A=$Particles.Timeline.Float, B=0, Comparison Type=::NewEnumerator5)) {isSetParameters=true}
   &ScaleMeshSize(Scale Factor=&VectorFromFloat(Value=$Particles.Timeline.Float)) {module=ScaleMeshSize001}
   &ScaleMeshSize(Scale Factor=&VectorFromFloat(Value=1)) {enabled=false}
   &PartitionParticles(
    Distance=65
    End Position:Vector3f=$Particles.SkeletalMeshLocation.SampledPosition
    Partition Mode:ENiagara_PartitionMode=::NewEnumerator3
    Start Position:NiagaraPosition=$Particles.CollisionPos
    UseDistanceFalloff=false
   ) {enabled=false}
   &self:SamplingAllBonesForPartition(
    ColPosOffset:Vector3f=&SelectVectorFromArray(Direct Array Index=$Particles.SpawnGroup, Vector Selection Array=DataInterface() {propertyValues {FloatData [<0 0 75>, <0 0 45>, <0 0 25>]}})
    Distance=35
    End Pos:NiagaraPosition=$Output.SocketLocation001.SampledPosition
    End Pos 02:NiagaraPosition=$Output.SocketLocation002.SampledPosition
    End Pos 03:NiagaraPosition=$Output.SocketLocation002.SampledPosition
    End Pos 04:NiagaraPosition=$Output.SocketLocation003.SampledPosition
    End Position:Vector3f=$Particles.SkeletalMeshLocation001.SampledPosition
    EndPos 05:NiagaraPosition=$Output.SocketLocation.SampledPosition
    Start Pos:NiagaraPosition=$Particles.CollisionPos
    Start Position:NiagaraPosition=$Particles.CollisionPos
   ) {module=SamplingAllBonesForPartition}
   &self:DeathSwitch_0_to_1(
    BuildinSizeArray:NiagaraDataInterfaceArrayInt32=$System.SizeArray
    BuildingTypeSpawnGroup:NiagaraInt32=$Particles.BuildingSizePartition
    IsOutsideMask:NiagaraBool=&Boolean_NotOperation(Bool=&Boolean_OrOperation(Bool1=$Particles.RegionMask.IsInsideMask, Bool2=$Particles.RegionMask001.IsInsideMask))
    Partition:NiagaraInt32=$Particles.SamplingAllBonesForPartition.Partition
    Position Array:NiagaraDataInterfaceArrayPosition=$System.DeathArray
   ) {module=DeathSwitch_0_to_1}
   &self:ScratchModule_03(DeadSwitch:NiagaraInt32=$Particles.DeadSwitch, VisibilityTag:NiagaraInt32=$Particles.VisibilityTag) {module=ScratchModule_03}
   &Color(Color=&SelectLinearColorFromArray(
     Array Sampling Mode=::NewEnumerator1
     Direct Array Index=$Particles.BuildingSizePartition
     Color Selection Array=DataInterface() {propertyValues {ColorData [<rgba 1 0 0.2475 1>, <rgba 1 0.725 0 1>, <rgba 0 1 0.2863 1>]}}
     Random Seed=?
    )) {enabled=false}
   &Color(Color=&MakeCustomLinearColorFromBool(Bool=$Particles.RegionMask.IsInsideMask, True Color=<rgba 1 0.0536 0 1>, False Color=<rgba 0.0483 1 0 1>)) {enabled=false, module=Color001}
  ]
  rendererClasses [&NiagaraMeshRendererProperties]
  renderers [
   {rendererClass=&NiagaraMeshRendererProperties, rendererIndex=0}
   {rendererClass=&NiagaraMeshRendererProperties, rendererIndex=1}
  ]
 }
 Chunks {
  emitterUpdateScript [
   &EmitterState()
   &SpawnPerFrame(
    Spawn:NiagaraBool=$System.DeathSpawn
    Spawn Count:NiagaraInt32=&SelectIntFromArray(Direct Array Index=$System.BuildingSIze, Int Selection Array=DataInterface() {propertyValues {IntData [30,15,8]}})
    Use Spawn Probability=false
   )
  ]
  particleSpawnScript [
   &InitializeParticle(
    Lifetime Max=1
    Lifetime Min=0.5
    Lifetime Mode=::NewEnumerator1
    Mesh Renderer Array Visibility Mode=::NewEnumerator3
    Mesh Renderer Info=DataInterface() {propertyValues {MeshRenderer=&self:Chunks_2.NiagaraMeshRendererProperties_0}}
    Mesh Renderer Random Seed=35
    Mesh Scale Mode=::NewEnumerator1
    Mesh Uniform Scale Max=0.85
    Mesh Uniform Scale Min=0.35
    Position=$System.DeathPosition
    Position Mode=::NewEnumerator1
    Sprite Size Mode=::NewEnumerator1
    Uniform Sprite Size Max=6
    Uniform Sprite Size Min=3
    UsePositionOffset=false
    Write Material Random=true
   )
   &FN_SetSpawnGroup()
   &AddVelocity(
    Cone Angle=180
    Cone Axis=<0 0 1>
    Distribution Along Cone Axis=0.2
    Module.UseConeFalloff=false
    Module.UseVelDistribution=true
    Override Seed=false
    Velocity Mode=::NewEnumerator2
    Velocity Speed=&Multiply_Float(
     A=&RandomRangeFloat(Minimum=50, Maximum=100)
     B=&SelectFloatFromArray(Direct Array Index=$System.BuildingSIze, Float Selection Array=DataInterface() {propertyValues {FloatData [1.3,0.6,0.3]}})
    )
   )
   &ShapeLocation(
    Box Midpoint=<0 0 0>
    Box Size=&SelectVectorFromArray(Direct Array Index=$System.BuildingSIze, Vector Selection Array=DataInterface() {propertyValues {FloatData [<5 5 75>, <5 5 40>, <5 5 15>]}})
    MaskSpawnGroup=false
    Override Seed=false
    Shape Primitive=::NewEnumerator2
    Surface Only=false
    UseOwnerScale=true
   )
   &InitialMeshOrientation(Use Rotation Vector=false)
   /SetVariables(Particles.VisibilityTag=&SelectIntFromArray(Direct Array Index=$Particles.SpawnGroup, Int Selection Array=DataInterface() {propertyValues {IntData [0,1,1,1]}})) {isSetParameters=true}
  ]
  particleUpdateScript [
   &ParticleState()
   &ScaleSpriteSize(Uniform Curve Sprite Scale=DataInterface() {propertyValues {Curve {keys [
        <at 0 1>
        {arriveTangent=-1, interpMode=RCIM_Cubic, leaveTangent=-1, time=0.8, value=1}
        {arriveTangent=-5.7307, interpMode=RCIM_Cubic, leaveTangent=-5.7307, tangentMode=RCTM_User, time=1, value=0}
       ]}}})
   &ScaleMeshSize(Scale Factor=&VectorFromFloat(Value=&FloatFromCurve(FloatCurve=DataInterface() {propertyValues {Curve {keys [
          <at 0 1>
          {arriveTangent=-1, interpMode=RCIM_Cubic, leaveTangent=-1, time=0.8, value=1}
          {arriveTangent=-5.7307, interpMode=RCIM_Cubic, leaveTangent=-5.7307, tangentMode=RCTM_User, time=1, value=0}
         ]}}})))
   &GravityForce(Gravity=<0 0 -700>)
   &Drag(Drag=&RandomRangeFloat(Minimum=1, Maximum=2), Use Linear Drag=true, Use Rotational Drag=false)
   &ScaleColor(
    Scale Alpha=&FloatFromCurve(FloatCurve=DataInterface() {propertyValues {Curve {keys [<at -0.0046 1.0113 RCIM_Cubic>,<at 1.0011 0.002 RCIM_Cubic>]}}})
    ScaleA=true
    ScaleRGB=true
   ) {enabled=false}
   &Collision(
    Age Colliding Particles=false
    KillOnCollision=false
    Module.Control Rest State Via Surface Normal=false
    Module.EnableMaxCollisionCount=false
    Module.Randomize Collision Normal Vector=false
    Radius Calculation Type=::NewEnumerator1
   )
   &SolveForcesAndVelocity(Clamp Velocity=false, Limit Acceleration=false)
   &DynamicMaterialParameters(
    Index 0 Param 1=&Add_Float(
     A=&RandomRangeFloat(Minimum=0, Maximum=1)
     B=&FloatFromCurve(
      FloatCurve=DataInterface() {propertyValues {Curve {keys [
          <at 0 0 RCIM_Cubic>
          {arriveTangent=1.5947, interpMode=RCIM_Cubic, leaveTangent=1.5947, tangentMode=RCTM_User, time=0.2557, value=0.5257}
          <at 1 1 RCIM_Cubic>
         ]}}}
      CurveIndex=$Particles.Age
      Scale Curve=0.25
     )
    )
    Module.Param0WriteEnabled=true
    Module.Param1WriteEnabled=true
    Module.Param2WriteEnabled=true
    Module.Param3WriteEnabled=true
   ) {module=DynamicMaterialParameters001}
   &FN_Sprite_Attributes(Color=<rgba 0.2031 0.2031 0.2031 1>, Create Color Attribute=true, Write to Module Namespace=true)
  ]
  rendererClasses [&NiagaraMeshRendererProperties, &NiagaraSpriteRendererProperties]
  renderers [
   {rendererClass=&NiagaraMeshRendererProperties, rendererIndex=0}
   {rendererClass=&NiagaraSpriteRendererProperties, rendererIndex=1}
  ]
 }
 NE_Empty {
  bEnabled=false
  emitterUpdateScript [
   &EmitterState(Life Cycle Mode=::NewEnumerator1, Loop Duration=1, UseLoopDelay=false)
   &SpawnRate(SpawnRate=20, Use Spawn Probability=false)
  ]
  particleSpawnScript [
   /Niagara/Modules/Spawn/Initialization/InitializeParticle(
    Lifetime=5
    Module.Write Color=true
    Module.Write Lifetime=true
    Module.Write Mass=true
    Module.Write Material Random=true
    Module.Write Position=false
    Module.Write Scale=false
    Module.Write SpriteRotation=false
    Module.Write SpriteSize=true
    Position=/Niagara/Functions/Localspace/SimulationPosition()
    Sprite Size=<25 25>
   ) {enabled=false}
   /Niagara/Modules/Spawn/Initialization/V2/InitializeParticle(Sprite Size Mode=::NewEnumerator3, Uniform Sprite Size=25, UsePositionOffset=true, Write Material Random=true) {module=InitializeParticle001}
   &SkeletalMeshLocation(Bone Sampling Mode=::NewEnumerator4, MaskSpawnGroup=false, Sampled Position Offset=<0 0 75>, Skeletal Mesh=$System.Skeletal Mesh)
  ]
  particleUpdateScript [
   &ParticleState()
   &SolveForcesAndVelocity(Clamp Velocity=false, Limit Acceleration=false)
  ]
  rendererClasses [&NiagaraSpriteRendererProperties]
  renderers [{rendererClass=&NiagaraSpriteRendererProperties}]
 }
 Smoke {
  emitterUpdateScript [
   &EmitterState()
   &SpawnPerFrame(
    Spawn:NiagaraBool=$System.DeathSpawn
    Spawn Count:NiagaraInt32=&SelectIntFromArray(Direct Array Index=$System.BuildingSIze, Int Selection Array=DataInterface() {propertyValues {IntData [7,5,3]}})
    Use Spawn Probability=false
   )
  ]
  particleSpawnScript [
   &InitializeParticle(
    Color=<rgba 0.2795 0.2795 0.2795 0.7>
    Lifetime Max=1.75
    Lifetime Min=0.8
    Lifetime Mode=::NewEnumerator1
    Mass Max=2
    Mass Mode=::NewEnumerator3
    Position=$System.DeathPosition
    Position Mode=::NewEnumerator1
    Sprite Rotation Mode=::NewEnumerator2
    Sprite Size Mode=::NewEnumerator1
    Uniform Sprite Size Max=70
    Uniform Sprite Size Min=40
    UsePositionOffset=false
    Write Material Random=true
   )
   &AddVelocity(
    Cone Angle=30
    Cone Axis=<0 0 1>
    Distribution Along Cone Axis=0.2
    Module.UseConeFalloff=false
    Module.UseVelDistribution=true
    Override Seed=false
    Velocity Mode=::NewEnumerator2
    Velocity Speed=&Multiply_Float(
     A=&RandomRangeFloat(Minimum=100, Maximum=300)
     B=&SelectFloatFromArray(Direct Array Index=$System.BuildingSIze, Float Selection Array=DataInterface() {propertyValues {FloatData [1.3,0.6,0.3]}})
    )
   )
   &ShapeLocation(
    Box Midpoint=<0 0 0>
    Box Size=&SelectVectorFromArray(Direct Array Index=$System.BuildingSIze, Vector Selection Array=DataInterface() {propertyValues {FloatData [<5 5 75>, <5 5 40>, <5 5 15>]}})
    MaskSpawnGroup=false
    Override Seed=false
    Shape Primitive=::NewEnumerator2
    Surface Only=false
    UseOwnerScale=true
   )
   &Color(Color=&SelectLinearColorFromArray(
     Array Sampling Mode=::NewEnumerator1
     Direct Array Index=$System.BuildingSIze
     Color Selection Array=DataInterface() {propertyValues {ColorData [<rgba 1 0 0.1786 1>, <rgba 1 0.9578 0 1>, <rgba 0 1 0.2555 1>]}}
     Random Seed=?
    )) {enabled=false}
  ]
  particleUpdateScript [
   &ParticleState()
   &ScaleSpriteSize(Uniform Curve Sprite Scale=DataInterface() {propertyValues {Curve {keys [
        {arriveTangent=2.087, interpMode=RCIM_Cubic, leaveTangent=2.087, tangentMode=RCTM_User, time=0, value=0.4}
        {interpMode=RCIM_Cubic, tangentMode=RCTM_User, time=1, value=1}
       ]}}})
   &GravityForce(Gravity=<0 0 -300>)
   &Drag(Drag=10, Use Linear Drag=true, Use Rotational Drag=false)
   &ScaleColor(
    Scale Alpha=&FloatFromCurve(FloatCurve=DataInterface() {propertyValues {Curve {keys [<at -0.0046 1.0113 RCIM_Cubic>,<at 1.0011 0.002 RCIM_Cubic>]}}})
    ScaleA=true
    ScaleRGB=true
   )
   &SolveForcesAndVelocity(Clamp Velocity=false, Limit Acceleration=false)
   &DynamicMaterialParameters(
    Index 0 Param 1=&Add_Float(
     A=&RandomRangeFloat(Minimum=0, Maximum=1)
     B=&FloatFromCurve(
      FloatCurve=DataInterface() {propertyValues {Curve {keys [
          <at 0 0 RCIM_Cubic>
          {arriveTangent=1.5947, interpMode=RCIM_Cubic, leaveTangent=1.5947, tangentMode=RCTM_User, time=0.2557, value=0.5257}
          <at 1 1 RCIM_Cubic>
         ]}}}
      CurveIndex=$Particles.Age
      Scale Curve=0.25
     )
    )
    Module.Param0WriteEnabled=true
    Module.Param1WriteEnabled=true
    Module.Param2WriteEnabled=true
    Module.Param3WriteEnabled=true
   ) {module=DynamicMaterialParameters001}
  ]
  rendererClasses [&NiagaraSpriteRendererProperties]
  renderers [
   {rendererClass=&NiagaraSpriteRendererProperties}
   {rendererClass=&NiagaraSpriteRendererProperties, rendererIndex=1}
  ]
 }
}
summary {
 systemName=NS_Emote_BlondeJaw_02
 emitters [
  {emitterName=Buildings, rendererClasses [&NiagaraMeshRendererProperties]}
  {emitterName=Smoke, rendererClasses [&NiagaraSpriteRendererProperties]}
  {emitterName=Chunks, rendererClasses [&NiagaraMeshRendererProperties, &NiagaraSpriteRendererProperties]}
  {bEnabled=false, emitterName=NE_Empty, rendererClasses [&NiagaraSpriteRendererProperties]}
 ]
 userVariables [
  User.bIsFrontEnd {type=NiagaraBool}
  User.bIsFrontEndPreview {type=NiagaraBool}
 ]
}
