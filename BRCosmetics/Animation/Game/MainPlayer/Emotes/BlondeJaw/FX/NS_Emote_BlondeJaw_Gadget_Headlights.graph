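# Niagara system for the BlondeJaw emote gadget, providing headlights and spinning motion-blur effects, likely for rotors.
# Headlights spawns by time of day and drives a light and sprites, scaled by fade-in and time of day with a front-end override.
# Its flare also fades by camera angle through a dot-product curve.
# FN_moblur spins flickering, time-of-day-tinted sprites, MotionBlurFX rotates a mesh, and three other emitters have no renderers.
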NiagaraSystem /BRCosmetics/Animation/Game/MainPlayer/Emotes/BlondeJaw/FX/NS_Emote_BlondeJaw_Gadget_Headlights
dependencies {
 usedDataInterfaces [&NiagaraDataInterfaceCurve, &NiagaraDataInterfaceCamera, &NiagaraDataInterfaceColorCurve]
 usedDynamicInputs [&SimulationPosition, &ColorFromCurve]
 usedModules [
  &self:SystemSpawnScript.SystemScriptSource.SystemScriptGraph.NiagaraNodeAssignment_1.SetVariables_#1
  &SystemState
  &self:SystemSpawnScript.SystemScriptSource.SystemScriptGraph.NiagaraNodeAssignment_0.SetVariables_#2
  &self:SystemSpawnScript.SystemScriptSource.SystemScriptGraph.NiagaraNodeAssignment_3.SetVariables_#3
 ]
}
emitters {
 FN_moblur {
  emitterUpdateScript [
   &EmitterState(
    Inactive Response=::NewEnumerator1
    Life Cycle Mode=::NewEnumerator1
    Loop Behavior=::NewEnumerator1
    Loop Duration Mode=::NewEnumerator1
    UseLoopDelay=false
   )
   &SpawnBurst_Instantaneous(Spawn Count=1, Use Spawn Probability=false)
  ]
  particleSpawnScript [
   &InitializeParticle(
    Color=<rgba 0.276 0.0129 0.0312 1>
    Position Offset=<8 0 10>
    Sprite Size Mode=::NewEnumerator3
    Uniform Sprite Size=40
    UsePositionOffset=true
    Write Material Random=true
   ) {module=InitializeParticle001}
   /SetVariables(
    Particles.SecondaryColor:LinearColor=<rgba 0.1562 0.1562 0.1562 1>
    Particles.SecondarySize:Vector2f=&Vector2DFromFloat(Value=40)
   ) {isSetParameters=true}
  ]
  particleUpdateScript [
   &ParticleState(Kill Particles When Lifetime Has Elapsed=false, Let Infinitely Lived Particles Die When Emitter Deactivates=true)
   &SkeletalMeshLocation(
    Bone Sampling Mode=::NewEnumerator4
    Inherit Velocity=false
    Interpolate Mesh Sampling=false
    MaskSpawnGroup=false
    Sampled Position Offset Coordinate Space:ENiagaraMeshTransforms=::NewEnumerator2
    Skeletal Mesh=$System.Skeletal Mesh
   ) {enabled=false}
   &CameraOffset(Camera Offset Amount=2)
   /SetVariables(
    Particles.SpriteAlignment=/Niagara/DynamicInputs/Vectors/RotateVector(
     Vector To Rotate:Vector3f=<1 0 1>
     Yaw:NiagaraFloat=0
     Pitch:NiagaraFloat=-40
     Roll:NiagaraFloat=0
     Rotation Rate:NiagaraFloat=&Add_Float(A=0, B=1)
    )
    Particles.SpriteFacing=&RotateVector(
     Vector To Rotate:Vector3f=<0 0 1>
     Yaw:NiagaraFloat=0
     Pitch:NiagaraFloat=-40
     Roll:NiagaraFloat=0
     Rotation Rate:NiagaraFloat=&Add_Float(A=0, B=1)
    )
   ) {isSetParameters=true}
   &SpriteRotationRate(Rotation Rate=200)
   &ScaleColor(
    Scale Alpha=0.5
    Scale RGB=&VectorFromFloat(Value=&Multiply_Float(
      A=&FNTimeOfDayScalar(Night Scale=0.5, Morning Scale=0.5, Evening Scale=0.5)
      B=&Sine(Normalized Angle=$Emitter.Age, Period=0.1, Scale=3, Bias=3)
     ))
    ScaleA=true
    ScaleRGB=true
   )
   /SetVariables(Particles.SecondaryColor:LinearColor=&MultiplyLinearColor(
     Linear Color=$Particles.Initial.SecondaryColor
     Scale Factor=&Multiply_Vector4(
      A=&MakeVector4(
       X=1
       Y=1
       Z=1
       W=&Multiply_Float(A=&Sine(Normalized Angle=$Emitter.Age, Period=0.1, Scale=3, Bias=1), B=1)
      )
      B=&Vector4FromCurve(
       CurveIndex=Linked() {linkedVariable {name=NPC.FortniteNPC.TimeOfDay_0-1, type=NiagaraFloat}}
       Vector4Curve=DataInterface() {propertyValues {
         AlphaCurve {keys [<at 0 1>,<at 0.125 1>,<at 0.25 1>,<at 0.375 1>,<at 0.5 1>,<at 0.625 1>,<at 0.75 1>,<at 0.875 1>,<at 1 1>]}
         BlueCurve {keys [<at 0.125 0.6771>,<at 0.166 0.6172>,<at 0.208 0.5395>,<at 0.25 0.6172>,<at 0.291 0.6939>,<at 0.333 0.7379>,<at 0.375 0.7835>,<at 0.416 0.7991>,<at 0.458 0.8148>,<at 0.5 0.8308>,<at 0.541 0.8148>,<at 0.583 0.7991>,<at 0.625 0.7991>,<at 0.666 0.7529>,<at 0.708 0.7084>,<at 0.75 0.5711>,<at 0.791 0.4735>,<at 0.833 0.3231>,<at 0.8847 0.6771>]}
         GreenCurve {keys [<at 0.125 0.268>,<at 0.166 0.5333>,<at 0.208 0.5271>,<at 0.25 0.6376>,<at 0.291 0.7529>,<at 0.333 0.7835>,<at 0.375 0.8148>,<at 0.416 0.7991>,<at 0.458 0.8148>,<at 0.5 0.8469>,<at 0.541 0.7991>,<at 0.583 0.7991>,<at 0.625 0.8148>,<at 0.666 0.7991>,<at 0.708 0.7529>,<at 0.75 0.5711>,<at 0.791 0.5029>,<at 0.833 0.5029>,<at 0.8847 0.268>]}
         RedCurve {keys [<at 0.125 0.1375>,<at 0.166 0.7084>,<at 0.208 0.6867>,<at 0.25 0.8148>,<at 0.291 0.8879>,<at 0.333 0.8963>,<at 0.375 0.9131>,<at 0.416 0.8714>,<at 0.458 0.8714>,<at 0.5 0.9131>,<at 0.541 0.8632>,<at 0.583 0.8714>,<at 0.625 0.9131>,<at 0.666 0.8963>,<at 0.708 0.8714>,<at 0.75 0.7529>,<at 0.791 0.7529>,<at 0.833 0.8469>,<at 0.8847 0.1375>]}
        }}
      )
     )
    )) {isSetParameters=true}
  ]
  rendererClasses [&NiagaraSpriteRendererProperties]
  renderers [
   {rendererClass=&NiagaraSpriteRendererProperties}
   {rendererClass=&NiagaraSpriteRendererProperties, rendererIndex=1}
  ]
 }
 FX_Rotor_Blur {rendererClasses [], renderers []}
 Headlights {
  emitterUpdateScript [
   &EmitterState(Life Cycle Mode=::NewEnumerator1, Loop Behavior=::NewEnumerator1, Loop Duration=0.1, UseLoopDelay=false)
   &SpawnBurst_Instantaneous(
    Spawn Count=&MakeIntFromFloat(Float=&FNTimeOfDayScalar(Night Scale=1, Morning Scale=0, Day Scale=0, Evening Scale=1))
    Use Spawn Probability=false
    UseLoopCountLimit=false
   )
  ]
  particleSpawnScript [
   /SetVariables(Particles.MirrorAxis:Vector3f=<1 1 1>) {isSetParameters=true}
   &InitializeParticle(
    Color=<rgba 0.7 0.73 1 1>
    Lifetime=2
    Mesh Scale Mode=::NewEnumerator3
    Position Offset=<5 0 -2>
    Position Offset Coordinate Space=::Local
    Sprite Size=<38.4948 132.2541>
    Sprite Size Mode=::NewEnumerator4
    UsePositionOffset=true
    Write Material Random=true
   )
   &SpriteFacingAndAlignment(Sprite Alignment=<0 0 1>, WriteAlignment=true, WriteFacing=false)
   &CameraOffset(Camera Offset Amount=10)
   &DynamicMaterialParameters(
    Index 0 Param 1=&Multiply_Float(A=-19, B=&MakeFloatFromVector(Vector=$Particles.MirrorAxis, Channel=::NewEnumerator0))
    Module.Param0WriteEnabled=true
    Module.Param1WriteEnabled=true
    Module.Param2WriteEnabled=false
    Module.Param3WriteEnabled=false
   )
   /SetVariables(Particles.BeamColor:LinearColor=&MultiplyLinearColor(Linear Color=$Particles.Color, Scale Factor=&MakeVector4_FromFloat(Float=0.5))) {isSetParameters=true}
   /SetVariables(
    Particles.FlareCameraOffset=10
    Particles.FlareColor=&MultiplyLinearColor(Linear Color=$Particles.Color, Scale Factor=&MakeVector4_FromFloat(Float=1.5))
    Particles.FlarePosition:NiagaraPosition=&AddVectorToPosition(Position=$Particles.Position, Vector=&Multiply_Vector(A=<0 0 0>, B=$Particles.MirrorAxis))
    Particles.FlareSize:Vector2f=<40 20>
   ) {isSetParameters=true}
   /SetVariables(
    Particles.LightColor=&MultiplyLinearColor(Linear Color=$Particles.Color, Scale Factor=&MakeVector4_FromFloat(Float=10))
    Particles.LightPosition:NiagaraPosition=&AddVectorToPosition(Position=$Particles.Position, Vector=&Multiply_Vector(A=<0 -70 50>, B=$Particles.MirrorAxis))
   ) {isSetParameters=true}
  ]
  particleUpdateScript [
   &ParticleState()
   /SetVariables(
    Particles.TimeOfDayHeadlightsIntensity:NiagaraFloat=&Multiply_Float(
     A=$System.FadeIn
     B=&FN_FrontEndFloatModifier(
      Gameplay=&FNTimeOfDayScalar(Night Scale=1, Morning Scale=0, Day Scale=0, Evening Scale=0)
      Front End=1
     )
    )
    Particles.TimeOfDayVisibilityTag:NiagaraInt32=&MakeIntFromFloat(Float=&FN_FrontEndFloatModifier(Gameplay=&FNTimeOfDayScalar(Night Scale=0, Morning Scale=1, Evening Scale=1), Front End=0))
   ) {isSetParameters=true}
   /SetVariables(
    Particles.BeamColor:LinearColor=&MultiplyLinearColor(
     Linear Color=$Particles.Initial.BeamColor
     Scale Factor=&MakeVector4_FromFloat(Float=$Particles.TimeOfDayHeadlightsIntensity)
    )
    Particles.Color=&MultiplyLinearColor(
     Linear Color=$Particles.Initial.Color
     Scale Factor=&MakeVector4_FromFloat(Float=$Particles.TimeOfDayHeadlightsIntensity)
    )
    Particles.FlareColor=&MultiplyLinearColor(
     Linear Color=$Particles.Initial.FlareColor
     Scale Factor=&MakeVector4_FromFloat(Float=&Multiply_Float(
       A=$Particles.TimeOfDayHeadlightsIntensity
       B=&FloatFromCurve(
        FloatCurve=DataInterface() {propertyValues {Curve {keys [<at -1 0>,<at 0 0>,<at 0.1 1>]}}}
        CurveIndex=&DotProduct(
         Vector A=&Multiply_VectorByFloat(
          Vector=&ReturnCameraVector(Camera Vectors=::NewEnumerator4, Coordinate Space=::Simulation, Origin Position=$Particles.Position)
          Float=-1
         )
         Vector B=&RotateVector(
          Vector To Rotate:Vector3f=<0 0 1>
          Yaw:NiagaraFloat=0
          Pitch:NiagaraFloat=-40
          Roll:NiagaraFloat=0
          Rotation Rate:NiagaraFloat=&Add_Float(A=1, B=0)
         )
        )
       )
      ))
    )
    Particles.GlowColor=&MultiplyLinearColor(
     Linear Color=$Particles.Initial.GlowColor
     Scale Factor=&MakeVector4_FromFloat(Float=$Particles.TimeOfDayHeadlightsIntensity)
    )
    Particles.LightColor=&MultiplyLinearColor(
     Linear Color=$Particles.Initial.LightColor
     Scale Factor=&MakeVector4_FromFloat(Float=$Particles.TimeOfDayHeadlightsIntensity)
    )
    Particles.TeardropGlowColor:LinearColor=&MultiplyLinearColor(
     Linear Color=$Particles.Initial.TeardropGlowColor
     Scale Factor=&MakeVector4_FromFloat(Float=$Particles.TimeOfDayHeadlightsIntensity)
    )
   ) {isSetParameters=true}
   /SetVariables(Particles.FlareAlignment:Vector3f=&RotateVector(
     Vector To Rotate:Vector3f=<0 0 1>
     Yaw:NiagaraFloat=0
     Pitch:NiagaraFloat=-40
     Roll:NiagaraFloat=0
     Rotation Rate:NiagaraFloat=&Add_Float(A=1, B=0)
    )) {enabled=false, isSetParameters=true}
   /SetVariables(Particles.FlareAlignment:Vector3f=<0.7967 0 1>) {enabled=false, isSetParameters=true}
  ]
  rendererClasses [&NiagaraLightRendererProperties, &NiagaraSpriteRendererProperties]
  renderers [
   {rendererClass=&NiagaraLightRendererProperties, rendererIndex=0}
   {rendererClass=&NiagaraSpriteRendererProperties, rendererIndex=1}
   {rendererClass=&NiagaraSpriteRendererProperties, rendererIndex=2}
   {rendererClass=&NiagaraSpriteRendererProperties, rendererIndex=3}
  ]
 }
 MotionBlurFX {
  emitterUpdateScript [
   &EmitterState(Life Cycle Mode=::NewEnumerator1, Loop Behavior=::NewEnumerator1, Loop Duration Mode=::NewEnumerator1, UseLoopDelay=false)
   &SpawnBurst_Instantaneous(Use Spawn Probability=false, UseLoopCountLimit=false)
  ]
  particleSpawnScript [
   &InitializeParticle(
    Color=<rgba 0.4219 0.1252 0.1628 0.5>
    Mesh Scale Mode=::NewEnumerator3
    Mesh Uniform Scale=0.4
    Position Offset=<9.3077 0 5.8358>
    UsePositionOffset=true
    Write Material Random=true
   )
   &InitialMeshOrientation(Mesh Orientation Mode=::NewEnumerator7, Rotation=<0 -0.6414 0>, Use Rotation Vector=true)
  ]
  particleUpdateScript [
   &ParticleState(
    Kill Particles When Lifetime Has Elapsed=false
    Let Infinitely Lived Particles Die When Emitter Deactivates=true
    Loop Particles Lifetime=true
   )
   &CameraOffset(Camera Offset Amount=2)
   &DynamicMaterialParameters(
    Index 0 Param 1=1
    Module.Param0WriteEnabled=true
    Module.Param1WriteEnabled=true
    Module.Param2WriteEnabled=true
    Module.Param3WriteEnabled=true
   )
   &UpdateMeshOrientation(Rotation Rate=3)
  ]
  rendererClasses [&NiagaraMeshRendererProperties]
  renderers [{rendererClass=&NiagaraMeshRendererProperties, rendererIndex=0}]
 }
 Motion_Blur_Rotor {rendererClasses [], renderers []}
 Motion_Blur_Rotor001 {rendererClasses [], renderers []}
}
summary {
 systemName=NS_Emote_BlondeJaw_Gadget_Headlights
 emitters [
  {emitterName=Headlights, rendererClasses [&NiagaraLightRendererProperties, &NiagaraSpriteRendererProperties]}
  {emitterName=FN_moblur, rendererClasses [&NiagaraSpriteRendererProperties]}
  {emitterName=FX_Rotor_Blur, rendererClasses []}
  {emitterName=Motion_Blur_Rotor001, rendererClasses []}
  {emitterName=Motion_Blur_Rotor, rendererClasses []}
  {emitterName=MotionBlurFX, rendererClasses [&NiagaraMeshRendererProperties]}
 ]
 userVariables [
  User.bIsFrontEnd {type=NiagaraBool}
  User.bIsFrontEndPreview {type=NiagaraBool}
 ]
}
